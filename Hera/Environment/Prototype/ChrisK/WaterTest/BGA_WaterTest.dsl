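; Prototype water-test Blueprint actor that raises its Cube component after play begins.
; On BeginPlay it stores the Cube's world location as the start and sets the end 512 units above it.
; After a six-second delay, the MoveMe timeline drives a vector lerp between the two points, setting the Cube's world location with teleport and no sweep.
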(event EventBeginPlay
  (bind _output_get (Variables|Default|SetStartLoc (Transformation|GetWorldLocation (Variables|Default|GetCube))))
  (Variables|Default|SetEndLoc (+ _output_get "0.000000,0.000000,512.000000"))
  (Utilities|FlowControl|Delay 6.0)
  (bind (_direction _lerp) (Timeline|MoveMe PlayFromStart)
    (:Update
      (Transformation|SetWorldLocation (Variables|Default|GetCube) (Math|Vector|Lerp(Vector) (Variables|Default|GetStartLoc) (Variables|Default|GetEndLoc) _lerp) false true))))

(fn ConstructionScript)
